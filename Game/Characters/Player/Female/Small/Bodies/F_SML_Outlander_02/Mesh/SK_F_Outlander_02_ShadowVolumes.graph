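# Physics asset of simplified shadow volumes for the small female Outlander 02 player body, apparently for capsule shadows.
# One capsule per body covers pelvis, upper spine, arms and legs down to the feet, with none for head or hands.
# Some limb capsules sit on twist bones, differing between left and right.
# Constraints share uniform 45-degree swing and twist limits, with some skipping the anatomical parent to link straight to spine or pelvis.
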PhysicsAsset /Game/Characters/Player/Female/Small/Bodies/F_SML_Outlander_02/Mesh/SK_F_Outlander_02_ShadowVolumes
bodies {
 calf_l [{shapeType=Capsule, length=42.3986, radius=6.7799, center=<-21.5437 -3.8519 -0.9467>, rotation=<rot 0 91.2975 -88.7715>}]
 calf_r [{shapeType=Capsule, length=46.9049, radius=6.7067, center=<20.4334 2.7526 0.3463>, rotation=<rot -0.0143 89.8902 -90.1098>}]
 foot_l [{shapeType=Capsule, length=9.9907, radius=7.1592, center=<-3.5911 5.6862 -0.3914>, rotation=<rot -0.6684 -178.9757 -91.3652>}]
 foot_r [{shapeType=Capsule, length=8.1961, radius=7.1592, center=<3.6135 -5.6344 0.6419>, rotation=<rot 0.6113 3.3954 -88.6084>}]
 lowerarm_l [{shapeType=Capsule, length=29.2308, radius=4.6755, center=<14.942 0.4274 -0.3302>, rotation=<rot 0.1714 -87.9862 -89.1044>}]
 lowerarm_r [{shapeType=Capsule, length=27.7284, radius=5.9418, center=<-13.3968 -1.5511 0.4872>, rotation=<rot 0 -86.4737 -88.5326>}]
 pelvis [{shapeType=Capsule, length=5.0913, radius=14.1606, center=<0.5964 -3.7425 0.1312>, rotation=<rot 4.5817 -86.1995 0.5976>}]
 spine_04 [{shapeType=Capsule, length=23.8637, radius=11.8044, center=<0.1253 2.1938 -0.004>, rotation=<rot 0 90 -90>}]
 thigh_r [{shapeType=Capsule, length=33.2153, radius=8.5083, center=<21.9641 -0.7911 1.5207>, rotation=<rot 0 -87.3298 -88.6388>}]
 thigh_twist_01_l [{shapeType=Capsule, length=37.2555, radius=8.8892, center=<6.6231 -0.5558 -0.8828>, rotation=<rot 0 89.7391 -89.4758>}]
 upperarm_l [{shapeType=Capsule, length=21.7504, radius=4.4373, center=<11.6465 0.1236 -0.0418>, rotation=<rot 0 84.7891 -87.5641>}]
 upperarm_twist_02_r [{shapeType=Capsule, length=20.8849, radius=4.2629, center=<10.8855 -0.4235 0.1509>, rotation=<rot -13.004 87.3364 -90.6573>}]
}
constraints [
 {bone1Name=spine_04, bone2Name=pelvis, swing1LimitDegrees=45, swing2LimitDegrees=45, twistLimitDegrees=45}
 {bone1Name=upperarm_l, bone2Name=spine_04, swing1LimitDegrees=45, swing2LimitDegrees=45, twistLimitDegrees=45}
 {bone1Name=lowerarm_l, bone2Name=upperarm_l, swing1LimitDegrees=45, swing2LimitDegrees=45, twistLimitDegrees=45}
 {bone1Name=lowerarm_r, bone2Name=spine_04, swing1LimitDegrees=45, swing2LimitDegrees=45, twistLimitDegrees=45}
 {bone1Name=upperarm_twist_02_r, bone2Name=spine_04, swing1LimitDegrees=45, swing2LimitDegrees=45, twistLimitDegrees=45}
 {bone1Name=calf_l, bone2Name=pelvis, swing1LimitDegrees=45, swing2LimitDegrees=45, twistLimitDegrees=45}
 {bone1Name=foot_l, bone2Name=calf_l, swing1LimitDegrees=45, swing2LimitDegrees=45, twistLimitDegrees=45}
 {bone1Name=thigh_twist_01_l, bone2Name=pelvis, swing1LimitDegrees=45, swing2LimitDegrees=45, twistLimitDegrees=45}
 {bone1Name=thigh_r, bone2Name=pelvis, swing1LimitDegrees=45, swing2LimitDegrees=45, twistLimitDegrees=45}
 {bone1Name=calf_r, bone2Name=thigh_r, swing1LimitDegrees=45, swing2LimitDegrees=45, twistLimitDegrees=45}
 {bone1Name=foot_r, bone2Name=calf_r, swing1LimitDegrees=45, swing2LimitDegrees=45, twistLimitDegrees=45}
]
